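# Megascans preset master material for fuzzy surfaces, using the Cloth shading model and material attributes.
# MF_Tiling supplies UVs to Albedo, Metalness, packed DRF and Normal texture parameters, plus a rotation angle for the normals.
# MF_MapAdjustments, shared from the default Megascans material, combines these maps with DRF green as roughness and a constant in place of AO.
# MF_Fuzz then takes that result and DRF blue as the fuzz map.
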Material /Bridge/MSPresets/M_MS_Default_Fuzz_Material/M_MS_Default_Fuzz_Material
graph {
 properties {bUseMaterialAttributes=true, pixelDepthOffsetMode=PDOM_Legacy, shadingModel=MSM_Cloth, usedShadingModels=Cloth}
 nodes [
  n1 = MaterialFunctionCall() {materialFunction=/Bridge/MSPresets/M_MS_Default_Fuzz_Material/Functions/MF_Tiling}
  n2 = TextureSampleParameter2D(UVs=@n1.Result) {
   group=07 - Texture Maps
   parameterName=Albedo
   samplerSource=SSM_Wrap_WorldGroupSettings
   sortPriority=0
   texture=/Engine/EngineMaterials/DefaultDiffuse
  }
  n3 = TextureSampleParameter2D(UVs=@n1.Result) {
   group=07 - Texture Maps
   parameterName=Metalness
   samplerSource=SSM_Wrap_WorldGroupSettings
   samplerType=SAMPLERTYPE_LinearColor
   sortPriority=1
   texture=/Bridge/MSPresets/MSTextures/BlackPlaceholder
  }
  n4 = TextureSampleParameter2D(UVs=@n1.Result) {
   desc=DRF
   group=07 - Texture Maps
   parameterName=DRF
   samplerType=SAMPLERTYPE_LinearColor
   sortPriority=2
   texture=/Bridge/MSPresets/MSTextures/WhitePlaceholder
  }
  n5 = TextureSampleParameter2D(UVs=@n1.Result) {
   group=07 - Texture Maps
   parameterName=Normal
   samplerSource=SSM_Wrap_WorldGroupSettings
   samplerType=SAMPLERTYPE_Normal
   sortPriority=3
   texture=/Bridge/MSPresets/MSTextures/Placeholder_Normal
  }
  MaterialFunctionCall!(
   MA=MaterialFunctionCall!(AO=Constant(), Albedo=@n2.RGB, Metallic=@n3.R, Roughness=@n4.G, Tangent Normal=@n5.RGB, Tangent Normal Rotation Angle=@n1.Rotation Angle) {materialFunction=/Bridge/MSPresets/M_MS_Default_Material/Functions/MF_MapAdjustments}.Result
   Fuzz Map=@n4.B
  ) {materialFunction=/Bridge/MSPresets/M_MS_Default_Fuzz_Material/Functions/MF_Fuzz}
 ]
}
